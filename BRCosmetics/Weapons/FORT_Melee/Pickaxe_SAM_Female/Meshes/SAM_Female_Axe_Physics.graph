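# Physics asset for the SAM Female pickaxe cosmetic mesh, defining a simulated dangling chain on the axe.
# A small capsule on dyn_master anchors five equally sized capsule bodies, dyn_01 through dyn_05, linked in series.
# Each joint uses the same constraint, allowing wide swing with very little twist so the chain bends without spinning along its length.
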PhysicsAsset /BRCosmetics/Weapons/FORT_Melee/Pickaxe_SAM_Female/Meshes/SAM_Female_Axe_Physics
bodies {
 dyn_01 [{shapeType=Capsule, length=4, radius=2, shapeName=dyn_01_capsule, center=<4 0 0>, rotation=<rot -90 0 0>}]
 dyn_02 [{shapeType=Capsule, length=4, radius=2, shapeName=dyn_02_capsule, center=<4 0 0>, rotation=<rot -90 0 0>}]
 dyn_03 [{shapeType=Capsule, length=4, radius=2, shapeName=dyn_03_capsule, center=<4 0 0>, rotation=<rot -90 0 0>}]
 dyn_04 [{shapeType=Capsule, length=4, radius=2, shapeName=dyn_04_capsule, center=<4 0 0>, rotation=<rot -90 0 0>}]
 dyn_05 [{shapeType=Capsule, length=4, radius=2, shapeName=dyn_05_capsule, center=<4 0 0>, rotation=<rot -90 0 0>}]
 dyn_master [{shapeType=Capsule, length=0.505, radius=0.505, shapeName=dyn_master_capsule}]
}
constraints [
 {bone1Name=dyn_01, bone2Name=dyn_master, swing1LimitDegrees=60, swing2LimitDegrees=60, twistLimitDegrees=5}
 {bone1Name=dyn_02, bone2Name=dyn_01, swing1LimitDegrees=60, swing2LimitDegrees=60, twistLimitDegrees=5}
 {bone1Name=dyn_03, bone2Name=dyn_02, swing1LimitDegrees=60, swing2LimitDegrees=60, twistLimitDegrees=5}
 {bone1Name=dyn_04, bone2Name=dyn_03, swing1LimitDegrees=60, swing2LimitDegrees=60, twistLimitDegrees=5}
 {bone1Name=dyn_05, bone2Name=dyn_04, swing1LimitDegrees=60, swing2LimitDegrees=60, twistLimitDegrees=5}
]
